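# MetaSound Source that plays looping stereo rollercoaster music for the Wonkeeland grind track coaster.
# A single stereo Wave Player starts on play and feeds its left and right channels straight to the output.
# Base submix and submix sends are disabled, so audio routes through a send to the Wonkeeland coaster audio bus.
# It uses the gameplay music sound class and keeps playing when silent.
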MetaSoundSource /GrindRail/Audio/GrindTrack/MSS_GrindTrack_CoasterMusic
inputs [UE.Source.OnPlay {type=Trigger}]
interfaces [UE.OutputFormat.Stereo:1.0, UE.Source:1.1]
outputs [
 "UE.OutputFormat.Stereo.Audio:0" {type=Audio}
 "UE.OutputFormat.Stereo.Audio:1" {type=Audio}
]
pages [Default {
  input_defaults {UE.Source.OnPlay=null}
  nodes [
   InitVariable.WaveAsset()
   WavePlayer_0 = UE.Wave Player:Stereo(Play=^UE.Source.OnPlay) {also {
     defaults {Loop="<MetaSoundBuilderResult.SUCCEEDED: 0>", Wave Asset=&/MX_S39_Wonkeeland_Rollercoaster_LP}
     version [1,1]
    }}
  ]
  links [{dst="@UE.OutputFormat.Stereo.Audio:0", src=WavePlayer_0.Out Left}, {dst="@UE.OutputFormat.Stereo.Audio:1", src=WavePlayer_0.Out Right}]
 }]
properties {
 assetImportData=&self:AssetImportData
 bEnableBaseSubmix=false
 bEnableSubmixSends=false
 bLooping=true
 duration=10000
 numChannels=2
 outputFormat=Stereo
 sampleRate=48000
 soundClassObject=/Game/Sounds/FortSoundClassesAndModes/Fort_Music_Gameplay
 virtualizationMode=PlayWhenSilent
 volume=0.355
 busSends [{audioBus=&/AB_WonkeeLand_Coaster, sendLevel=1}]
}
